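# Animation Blueprint for the Logarithm glider, animating its deployment, gliding and leaning from skydiving state.
# A deploy state machine with nested forward thrust plus additive forward-lean start/stop drives gameplay, while cosmetic previews use a FrontEnd machine.
# A curve-weighted glide additive, then mesh-space turn and lean blendspaces, layer on top.
# Notifies mark full deployment and suppress the thrust animation when the glide loop begins already leaning forward.
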AnimBlueprint /BRCosmetics/Gadgets/Assets/VinderTech_GliderChute/Glider_Logarithm/Meshes/Logarithm_Glider_AnimBP
target_skeleton=/Game/Gadgets/Assets/VinderTech_GliderChute/Glider_Logarithm/Meshes/Logarithm_Glider_Skeleton
graphs [
 AnimGraph {nodes [
   AnimGraphNode_StateMachine_0 = New State Machine() {is=title AnimGraphNode_StateMachine}
   AnimGraphNode_StateMachine_1 = New State Machine_1() {is=title AnimGraphNode_StateMachine}
   AnimGraphNode_StateMachine_2 = FrontEnd() {is=title AnimGraphNode_StateMachine, bSkipFirstUpdateTransition=false}
   n1 = Get OwnerGlider() -> (OwnerGlider:Fort Player Parachute Object Reference)
   n2 = Break Fort Anim Input Skydiving!(FortAnimInput_Skydiving:"Fort Anim Input Skydiving Structure (by ref)"=Get SkydivingState() -> (SkydivingState:Fort Anim Input Skydiving Structure)) -> (
    ParaGilderRootOffset:Vector=~
    ParaGliderRootRotationOverride:Rotator=~
    bUseParaGlideRootModifier:bool=~
    bIsSkydivingFromLaunchPad:bool=~
    bIsSkydivingFromBus:bool=~
    bIsInVortex:bool=~
    bIsDiving:bool=~
    bIsParachuteOpen:bool=~
    bIsSkydiving:bool=~
    bIsParachuteLeaning:bool=~
    bIsSkydiveLeaning:bool=~
    bIsLeaning:bool=~
    bIsSkydiveDiveMode:bool=~
    bPlaySkydiveDrift:bool=~
    bSkydiveDriftDelayActive:bool=~
    bSkydiveDriftAnimAllowed:bool=~
    bIsGliderRight:bool=~
    bIsGliderCenter:bool=~
    bIsGliderLeft:bool=~
    bIsGliderForward:bool=~
    bIsGliderBack:bool=~
    bHasDeployedAndIsStandingUpright:bool=~
    LocalAccelForward:float=~
    LocalAccelRight:float=~
    LocalVelocityRight:float=~
    SkydiveAimPitch:float=~
    SkydiveAimYaw:float=~
    DeployChuteAnimRate:float=~
    SkydiveDriftAnimRate:float=~
    SkydiveDriftAnimRateCurrent:float=~
    SkydiveFidgetAnimRate:float=~
    SkydiveFidgetAnimRateCurrent:float=~
    SkydiveAdditiveAlpha:float=~
    SkydiveDriftDelay:float=~
    SkydiveDriftAnim:int
    SkydiveDriftAnimMax:int
    LaunchpadAnim:int
    LocalAccelDir:ESkydivingDirection Enum
    DirectionLast:ESkydivingDirection Enum
    PoseAdditive:Anim Sequence Object Reference
   ) {is=title K2Node_BreakStruct}
   n3 = Apply Additive!(
    Base=Blend Poses by bool(
     BlendPose_0=@AnimGraphNode_StateMachine_2
     BlendPose_1=Apply Additive!(Base=@AnimGraphNode_StateMachine_0, Additive=@AnimGraphNode_StateMachine_1, Alpha:float) -> (Pose) {alphaBoolBlend {blendOption=Linear}}
     bActiveValue=Get IsCosmeticPreview!(self:Fort Player Parachute Object Reference=@n1) -> (IsCosmeticPreview:bool)
    ) {blendType=Linear}
    Additive="Sequence Player (additive) 'Logarithm_Glide_GLIDER'"!(PlayRate:float=SelectFloat(A=1.9, B=1, bPickA=@n2.bIsGliderForward)) -> (Pose) {sequence=/BRCosmetics/Animation/Game/MainPlayer/Skydive/ParaGlide/Logarithm/Logarithm_Glide_GLIDER}
    AlphaCurveName:Name=Idle_Loop_Additive
   ) -> (Pose) {alphaInputType=Curve, alphaBoolBlend {blendOption=Linear}}
   n4 = Blendspace Player 'Logarithm_Turn_GLIDER_BS'(X=@n2.SkydiveAimYaw, Y=@n2.SkydiveAimPitch) {blendSpace=/BRCosmetics/Animation/Game/MainPlayer/Skydive/ParaGlide/Logarithm/Logarithm_Turn_GLIDER_BS}
   Output Pose(Result=Apply Mesh Space Additive!(
     Base:Pose Link Structure=Apply Mesh Space Additive(Base=@n3, Additive=@n4, Alpha=1) {alphaBoolBlend {blendOption=Linear}}
     Additive:Pose Link Structure=Blendspace Player 'Logarithm_Lean_GLIDER_BS'(X=@n2.LocalAccelRight, Y=@n2.LocalAccelForward) {blendSpace=/BRCosmetics/Animation/Game/MainPlayer/Skydive/ParaGlide/Logarithm/Logarithm_Lean_GLIDER_BS}
     Alpha:float=1
    ) -> (Pose:Pose Link Structure) {is=title AnimGraphNode_ApplyMeshSpaceAdditive, alphaBoolBlend {blendOption=Linear}}) {layerGroup=None}
  ]}
 New State Machine {
  entry=Closed
  path=AnimGraph.AnimGraphNode_StateMachine_0.New State Machine
  ids=id
  nodes [
   AnimStateNode_0 = Closed() {is=title AnimStateNode}
   AnimStateNode_1 = Into() {is=title AnimStateNode}
   AnimStateNode_2 = Idle_Loop() {is=title AnimStateNode, stateEntered {guid=#1, notifyName=Idle_Loop_Entered}}
   AnimStateTransitionNode_0 = AnimStateTransitionNode()
   AnimStateTransitionNode_1 = AnimStateTransitionNode() {crossfadeDuration=0}
   AnimStateTransitionNode_2 = AnimStateTransitionNode() {bAutomaticRuleBasedOnSequencePlayerInState=true, crossfadeDuration=0}
   AnimStateTransitionNode_3 = AnimStateTransitionNode() {crossfadeDuration=0}
   AnimStateTransitionNode_4 = AnimStateTransitionNode()
  ]
  links [
   {
    dst=Into
    rule=AnimGraph.AnimGraphNode_StateMachine_0.New State Machine.AnimStateTransitionNode_0.Transition
    src=Closed
    via=AnimStateTransitionNode_0
   }
   {
    dst=Closed
    rule=AnimGraph.AnimGraphNode_StateMachine_0.New State Machine.AnimStateTransitionNode_3.Transition
    src=Idle_Loop
    via=AnimStateTransitionNode_3
   }
   {
    dst=Closed
    rule=AnimGraph.AnimGraphNode_StateMachine_0.New State Machine.AnimStateTransitionNode_1.Transition
    src=Into
    via=AnimStateTransitionNode_1
   }
   {
    dst=Idle_Loop
    rule=AnimGraph.AnimGraphNode_StateMachine_0.New State Machine.AnimStateTransitionNode_2.Transition
    src=Into
    via=AnimStateTransitionNode_2
   }
   {
    dst=Idle_Loop
    rule=AnimGraph.AnimGraphNode_StateMachine_0.New State Machine.AnimStateTransitionNode_4.Transition
    src=Into
    via=AnimStateTransitionNode_4
   }
  ]
 }
 Closed {
  path=AnimGraph.AnimGraphNode_StateMachine_0.New State Machine.AnimStateNode_0.Closed
  nodes [Output Animation Pose(Result=Sequence Player 'Logarithm_Closed_GLIDER'() {sequence=/BRCosmetics/Animation/Game/MainPlayer/Skydive/ParaGlide/Logarithm/Logarithm_Closed_GLIDER})]
 }
 Into {
  path=AnimGraph.AnimGraphNode_StateMachine_0.New State Machine.AnimStateNode_1.Into
  nodes [Output Animation Pose(Result=Sequence Player 'Logarithm_Into_GLIDER'!(bLoopAnimation:bool=False) -> (Pose) {bLoopAnimation=false, sequence=/BRCosmetics/Animation/Game/MainPlayer/Skydive/ParaGlide/Logarithm/Logarithm_Into_GLIDER})]
 }
 Idle_Loop {
  path=AnimGraph.AnimGraphNode_StateMachine_0.New State Machine.AnimStateNode_2.Idle_Loop
  nodes [
   AnimGraphNode_StateMachine_0 = New State Machine()
   Output Animation Pose(Result=@AnimGraphNode_StateMachine_0)
  ]
 }
 New State Machine {
  entry=Idle_Loop
  path=AnimGraph.AnimGraphNode_StateMachine_0.New State Machine.AnimStateNode_2.Idle_Loop.AnimGraphNode_StateMachine_0.New State Machine
  ids=id
  nodes [
   AnimStateNode_0 = Idle_Loop()
   AnimStateNode_1 = Lean_Fwd_Thrust() {is=title AnimStateNode}
   AnimStateTransitionNode_3 = AnimStateTransitionNode()
   AnimStateTransitionNode_5 = AnimStateTransitionNode() {crossfadeDuration=0}
   AnimStateTransitionNode_7 = AnimStateTransitionNode() {crossfadeDuration=0.4}
  ]
  links [
   {
    dst=Lean_Fwd_Thrust
    rule=AnimGraph.AnimGraphNode_StateMachine_0.New State Machine.AnimStateNode_2.Idle_Loop.AnimGraphNode_StateMachine_0.New State Machine.AnimStateTransitionNode_5.Transition
    src=Idle_Loop
    via=AnimStateTransitionNode_5
   }
   {
    dst=Idle_Loop
    rule=AnimGraph.AnimGraphNode_StateMachine_0.New State Machine.AnimStateNode_2.Idle_Loop.AnimGraphNode_StateMachine_0.New State Machine.AnimStateTransitionNode_3.Transition
    src=Lean_Fwd_Thrust
    via=AnimStateTransitionNode_3
   }
   {
    dst=Idle_Loop
    rule=AnimGraph.AnimGraphNode_StateMachine_0.New State Machine.AnimStateNode_2.Idle_Loop.AnimGraphNode_StateMachine_0.New State Machine.AnimStateTransitionNode_7.Transition
    src=Lean_Fwd_Thrust
    via=AnimStateTransitionNode_7
   }
  ]
 }
 Idle_Loop {
  path=AnimGraph.AnimGraphNode_StateMachine_0.New State Machine.AnimStateNode_2.Idle_Loop.AnimGraphNode_StateMachine_0.New State Machine.AnimStateNode_0.Idle_Loop
  nodes [Output Animation Pose(Result=Sequence Player 'Logarithm_Glide_Pose_GLIDER'() {sequence=/BRCosmetics/Animation/Game/MainPlayer/Skydive/ParaGlide/Logarithm/Logarithm_Glide_Pose_GLIDER})]
 }
 Lean_Fwd_Thrust {
  path=AnimGraph.AnimGraphNode_StateMachine_0.New State Machine.AnimStateNode_2.Idle_Loop.AnimGraphNode_StateMachine_0.New State Machine.AnimStateNode_1.Lean_Fwd_Thrust
  nodes [Output Animation Pose(Result=Sequence Player 'Logarithm_LeanFwd_Thrust_GLIDER'!(bLoopAnimation:bool=False) -> (Pose) {bLoopAnimation=false, sequence=/BRCosmetics/Animation/Game/MainPlayer/Skydive/ParaGlide/Logarithm/Logarithm_LeanFwd_Thrust_GLIDER})]
 }
 Transition {
  path=AnimGraph.AnimGraphNode_StateMachine_0.New State Machine.AnimStateNode_2.Idle_Loop.AnimGraphNode_StateMachine_0.New State Machine.AnimStateTransitionNode_3.Transition
  nodes [Result(bCanEnterTransition=Get SkipThrustFwdPlay() -> (SkipThrustFwdPlay:bool))]
 }
 Transition {
  path=AnimGraph.AnimGraphNode_StateMachine_0.New State Machine.AnimStateNode_2.Idle_Loop.AnimGraphNode_StateMachine_0.New State Machine.AnimStateTransitionNode_5.Transition
  nodes [
   n1 = Get SkydivingState()
   n2 = Get SkipThrustFwdPlay()
   Result(bCanEnterTransition=AND Boolean(
     A=NOT Boolean(A=@n2)
     B=Break Fort Anim Input Skydiving(FortAnimInput_Skydiving=@n1) -> (
      ParaGilderRootOffset=~
      ParaGliderRootRotationOverride=~
      bUseParaGlideRootModifier=~
      bIsSkydivingFromLaunchPad=~
      bIsSkydivingFromBus=~
      bIsInVortex=~
      bIsDiving=~
      bIsParachuteOpen=~
      bIsSkydiving=~
      bIsParachuteLeaning=~
      bIsSkydiveLeaning=~
      bIsLeaning=~
      bIsSkydiveDiveMode=~
      bPlaySkydiveDrift=~
      bSkydiveDriftDelayActive=~
      bSkydiveDriftAnimAllowed=~
      bIsGliderRight=~
      bIsGliderCenter=~
      bIsGliderLeft=~
      bIsGliderForward=~
      bIsGliderBack=~
      bHasDeployedAndIsStandingUpright=~
      LocalAccelForward=~
      LocalAccelRight=~
      LocalVelocityRight=~
      SkydiveAimPitch=~
      SkydiveAimYaw=~
      DeployChuteAnimRate=~
      SkydiveDriftAnimRate=~
      SkydiveDriftAnimRateCurrent=~
      SkydiveFidgetAnimRate=~
      SkydiveFidgetAnimRateCurrent=~
      SkydiveAdditiveAlpha=~
      SkydiveDriftDelay=~
     ).bIsGliderForward
    ))
  ]
 }
 Transition {
  path=AnimGraph.AnimGraphNode_StateMachine_0.New State Machine.AnimStateNode_2.Idle_Loop.AnimGraphNode_StateMachine_0.New State Machine.AnimStateTransitionNode_7.Transition
  nodes [
   n1 = Get SkydivingState()
   Result(bCanEnterTransition=NOT Boolean(A=Break Fort Anim Input Skydiving(FortAnimInput_Skydiving=@n1) -> (
      ParaGilderRootOffset=~
      ParaGliderRootRotationOverride=~
      bUseParaGlideRootModifier=~
      bIsSkydivingFromLaunchPad=~
      bIsSkydivingFromBus=~
      bIsInVortex=~
      bIsDiving=~
      bIsParachuteOpen=~
      bIsSkydiving=~
      bIsParachuteLeaning=~
      bIsSkydiveLeaning=~
      bIsLeaning=~
      bIsSkydiveDiveMode=~
      bPlaySkydiveDrift=~
      bSkydiveDriftDelayActive=~
      bSkydiveDriftAnimAllowed=~
      bIsGliderRight=~
      bIsGliderCenter=~
      bIsGliderLeft=~
      bIsGliderForward=~
      bIsGliderBack=~
      bHasDeployedAndIsStandingUpright=~
      LocalAccelForward=~
      LocalAccelRight=~
      LocalVelocityRight=~
      SkydiveAimPitch=~
      SkydiveAimYaw=~
      DeployChuteAnimRate=~
      SkydiveDriftAnimRate=~
      SkydiveDriftAnimRateCurrent=~
      SkydiveFidgetAnimRate=~
      SkydiveFidgetAnimRateCurrent=~
      SkydiveAdditiveAlpha=~
      SkydiveDriftDelay=~
     ).bIsGliderForward))
  ]
 }
 Transition {
  path=AnimGraph.AnimGraphNode_StateMachine_0.New State Machine.AnimStateTransitionNode_0.Transition
  nodes [
   n1 = Get SkydivingState()
   Result(bCanEnterTransition=OR Boolean(
     A=Break Fort Anim Input Skydiving(FortAnimInput_Skydiving=@n1) -> (
      ParaGilderRootOffset=~
      ParaGliderRootRotationOverride=~
      bUseParaGlideRootModifier=~
      bIsSkydivingFromLaunchPad=~
      bIsSkydivingFromBus=~
      bIsInVortex=~
      bIsDiving=~
      bIsParachuteOpen=~
      bIsSkydiving=~
      bIsParachuteLeaning=~
      bIsSkydiveLeaning=~
      bIsLeaning=~
      bIsSkydiveDiveMode=~
      bPlaySkydiveDrift=~
      bSkydiveDriftDelayActive=~
      bSkydiveDriftAnimAllowed=~
      bIsGliderRight=~
      bIsGliderCenter=~
      bIsGliderLeft=~
      bIsGliderForward=~
      bIsGliderBack=~
      bHasDeployedAndIsStandingUpright=~
      LocalAccelForward=~
      LocalAccelRight=~
      LocalVelocityRight=~
      SkydiveAimPitch=~
      SkydiveAimYaw=~
      DeployChuteAnimRate=~
      SkydiveDriftAnimRate=~
      SkydiveDriftAnimRateCurrent=~
      SkydiveFidgetAnimRate=~
      SkydiveFidgetAnimRateCurrent=~
      SkydiveAdditiveAlpha=~
      SkydiveDriftDelay=~
     ).bIsParachuteOpen
     B=Get bForceOpen() -> (bForceOpen:bool)
    ))
  ]
 }
 Transition {
  path=AnimGraph.AnimGraphNode_StateMachine_0.New State Machine.AnimStateTransitionNode_1.Transition
  nodes [
   n1 = Get bForceOpen()
   n2 = Break Fort Anim Input Skydiving(FortAnimInput_Skydiving=Get SkydivingState()) -> (
    ParaGilderRootOffset=~
    ParaGliderRootRotationOverride=~
    bUseParaGlideRootModifier=~
    bIsSkydivingFromLaunchPad=~
    bIsSkydivingFromBus=~
    bIsInVortex=~
    bIsDiving=~
    bIsParachuteOpen=~
    bIsSkydiving=~
    bIsParachuteLeaning=~
    bIsSkydiveLeaning=~
    bIsLeaning=~
    bIsSkydiveDiveMode=~
    bPlaySkydiveDrift=~
    bSkydiveDriftDelayActive=~
    bSkydiveDriftAnimAllowed=~
    bIsGliderRight=~
    bIsGliderCenter=~
    bIsGliderLeft=~
    bIsGliderForward=~
    bIsGliderBack=~
    bHasDeployedAndIsStandingUpright=~
    LocalAccelForward=~
    LocalAccelRight=~
    LocalVelocityRight=~
    SkydiveAimPitch=~
    SkydiveAimYaw=~
    DeployChuteAnimRate=~
    SkydiveDriftAnimRate=~
    SkydiveDriftAnimRateCurrent=~
    SkydiveFidgetAnimRate=~
    SkydiveFidgetAnimRateCurrent=~
    SkydiveAdditiveAlpha=~
    SkydiveDriftDelay=~
   )
   Result(bCanEnterTransition=AND Boolean(A=NOT Boolean(A=@n2.bIsParachuteOpen), B=NOT Boolean(A=@n1)))
  ]
 }
 Transition {path=AnimGraph.AnimGraphNode_StateMachine_0.New State Machine.AnimStateTransitionNode_2.Transition, nodes [Result()]}
 Transition {
  path=AnimGraph.AnimGraphNode_StateMachine_0.New State Machine.AnimStateTransitionNode_3.Transition
  nodes [
   n1 = Get bForceOpen()
   n2 = Break Fort Anim Input Skydiving(FortAnimInput_Skydiving=Get SkydivingState()) -> (
    ParaGilderRootOffset=~
    ParaGliderRootRotationOverride=~
    bUseParaGlideRootModifier=~
    bIsSkydivingFromLaunchPad=~
    bIsSkydivingFromBus=~
    bIsInVortex=~
    bIsDiving=~
    bIsParachuteOpen=~
    bIsSkydiving=~
    bIsParachuteLeaning=~
    bIsSkydiveLeaning=~
    bIsLeaning=~
    bIsSkydiveDiveMode=~
    bPlaySkydiveDrift=~
    bSkydiveDriftDelayActive=~
    bSkydiveDriftAnimAllowed=~
    bIsGliderRight=~
    bIsGliderCenter=~
    bIsGliderLeft=~
    bIsGliderForward=~
    bIsGliderBack=~
    bHasDeployedAndIsStandingUpright=~
    LocalAccelForward=~
    LocalAccelRight=~
    LocalVelocityRight=~
    SkydiveAimPitch=~
    SkydiveAimYaw=~
    DeployChuteAnimRate=~
    SkydiveDriftAnimRate=~
    SkydiveDriftAnimRateCurrent=~
    SkydiveFidgetAnimRate=~
    SkydiveFidgetAnimRateCurrent=~
    SkydiveAdditiveAlpha=~
    SkydiveDriftDelay=~
   )
   Result(bCanEnterTransition=AND Boolean(A=NOT Boolean(A=@n2.bIsParachuteOpen), B=NOT Boolean(A=@n1)))
  ]
 }
 Transition {
  path=AnimGraph.AnimGraphNode_StateMachine_0.New State Machine.AnimStateTransitionNode_4.Transition
  nodes [
   n1 = Get SkydivingState()
   n2 = "Get Relevant Anim Time (Into)"!() -> (ReturnValue:float) {is=title K2Node_AnimGetter}
   Result(bCanEnterTransition=AND Boolean(
     A=Break Fort Anim Input Skydiving(FortAnimInput_Skydiving=@n1) -> (
      ParaGilderRootOffset=~
      ParaGliderRootRotationOverride=~
      bUseParaGlideRootModifier=~
      bIsSkydivingFromLaunchPad=~
      bIsSkydivingFromBus=~
      bIsInVortex=~
      bIsDiving=~
      bIsParachuteOpen=~
      bIsSkydiving=~
      bIsParachuteLeaning=~
      bIsSkydiveLeaning=~
      bIsLeaning=~
      bIsSkydiveDiveMode=~
      bPlaySkydiveDrift=~
      bSkydiveDriftDelayActive=~
      bSkydiveDriftAnimAllowed=~
      bIsGliderRight=~
      bIsGliderCenter=~
      bIsGliderLeft=~
      bIsGliderForward=~
      bIsGliderBack=~
      bHasDeployedAndIsStandingUpright=~
      LocalAccelForward=~
      LocalAccelRight=~
      LocalVelocityRight=~
      SkydiveAimPitch=~
      SkydiveAimYaw=~
      DeployChuteAnimRate=~
      SkydiveDriftAnimRate=~
      SkydiveDriftAnimRateCurrent=~
      SkydiveFidgetAnimRate=~
      SkydiveFidgetAnimRateCurrent=~
      SkydiveAdditiveAlpha=~
      SkydiveDriftDelay=~
     ).bIsGliderForward
     B="float > float"(A=@n2, B=3.6)
    ))
  ]
 }
 New State Machine_1 {
  entry=Identity
  path=AnimGraph.AnimGraphNode_StateMachine_1.New State Machine_1
  ids=id
  nodes [
   AnimStateNode_0 = Identity() {is=title AnimStateNode}
   AnimStateNode_1 = Lean_Fwd_Start() {is=title AnimStateNode}
   AnimStateNode_2 = Lean_Fwd_Stop() {is=title AnimStateNode}
   AnimStateTransitionNode_0 = AnimStateTransitionNode()
   AnimStateTransitionNode_2 = AnimStateTransitionNode()
   AnimStateTransitionNode_3 = AnimStateTransitionNode() {bAutomaticRuleBasedOnSequencePlayerInState=true}
   AnimStateTransitionNode_4 = AnimStateTransitionNode()
  ]
  links [
   {
    dst=Lean_Fwd_Start
    rule=AnimGraph.AnimGraphNode_StateMachine_1.New State Machine_1.AnimStateTransitionNode_0.Transition
    src=Identity
    via=AnimStateTransitionNode_0
   }
   {
    dst=Lean_Fwd_Stop
    rule=AnimGraph.AnimGraphNode_StateMachine_1.New State Machine_1.AnimStateTransitionNode_2.Transition
    src=Lean_Fwd_Start
    via=AnimStateTransitionNode_2
   }
   {
    dst=Identity
    rule=AnimGraph.AnimGraphNode_StateMachine_1.New State Machine_1.AnimStateTransitionNode_3.Transition
    src=Lean_Fwd_Stop
    via=AnimStateTransitionNode_3
   }
   {
    dst=Identity
    rule=AnimGraph.AnimGraphNode_StateMachine_1.New State Machine_1.AnimStateTransitionNode_4.Transition
    src=Lean_Fwd_Stop
    via=AnimStateTransitionNode_4
   }
  ]
 }
 Identity {
  path=AnimGraph.AnimGraphNode_StateMachine_1.New State Machine_1.AnimStateNode_0.Identity
  nodes [Output Animation Pose(Result=Additive Identity Pose())]
 }
 Lean_Fwd_Start {
  path=AnimGraph.AnimGraphNode_StateMachine_1.New State Machine_1.AnimStateNode_1.Lean_Fwd_Start
  nodes [Output Animation Pose(Result="Sequence Player (additive) 'Logarithm_LeanFwd_Start_GLIDER'"!(bLoopAnimation:bool=False) -> (Pose) {bLoopAnimation=false, sequence=/BRCosmetics/Animation/Game/MainPlayer/Skydive/ParaGlide/Logarithm/Logarithm_LeanFwd_Start_GLIDER})]
 }
 Lean_Fwd_Stop {
  path=AnimGraph.AnimGraphNode_StateMachine_1.New State Machine_1.AnimStateNode_2.Lean_Fwd_Stop
  nodes [Output Animation Pose(Result="Sequence Player (additive) 'Logarithm_LeanFwd_Stop_GLIDER'"!(bLoopAnimation:bool=False) -> (Pose) {bLoopAnimation=false, sequence=/BRCosmetics/Animation/Game/MainPlayer/Skydive/ParaGlide/Logarithm/Logarithm_LeanFwd_Stop_GLIDER})]
 }
 Transition {
  path=AnimGraph.AnimGraphNode_StateMachine_1.New State Machine_1.AnimStateTransitionNode_0.Transition
  nodes [Result(bCanEnterTransition=Break Fort Anim Input Skydiving(FortAnimInput_Skydiving=Get SkydivingState()) -> (
     ParaGilderRootOffset=~
     ParaGliderRootRotationOverride=~
     bUseParaGlideRootModifier=~
     bIsSkydivingFromLaunchPad=~
     bIsSkydivingFromBus=~
     bIsInVortex=~
     bIsDiving=~
     bIsParachuteOpen=~
     bIsSkydiving=~
     bIsParachuteLeaning=~
     bIsSkydiveLeaning=~
     bIsLeaning=~
     bIsSkydiveDiveMode=~
     bPlaySkydiveDrift=~
     bSkydiveDriftDelayActive=~
     bSkydiveDriftAnimAllowed=~
     bIsGliderRight=~
     bIsGliderCenter=~
     bIsGliderLeft=~
     bIsGliderForward=~
     bIsGliderBack=~
     bHasDeployedAndIsStandingUpright=~
     LocalAccelForward=~
     LocalAccelRight=~
     LocalVelocityRight=~
     SkydiveAimPitch=~
     SkydiveAimYaw=~
     DeployChuteAnimRate=~
     SkydiveDriftAnimRate=~
     SkydiveDriftAnimRateCurrent=~
     SkydiveFidgetAnimRate=~
     SkydiveFidgetAnimRateCurrent=~
     SkydiveAdditiveAlpha=~
     SkydiveDriftDelay=~
    ).bIsGliderForward)]
 }
 Transition {
  path=AnimGraph.AnimGraphNode_StateMachine_1.New State Machine_1.AnimStateTransitionNode_2.Transition
  nodes [
   n1 = Get SkydivingState()
   Result(bCanEnterTransition=NOT Boolean(A=Break Fort Anim Input Skydiving(FortAnimInput_Skydiving=@n1) -> (
      ParaGilderRootOffset=~
      ParaGliderRootRotationOverride=~
      bUseParaGlideRootModifier=~
      bIsSkydivingFromLaunchPad=~
      bIsSkydivingFromBus=~
      bIsInVortex=~
      bIsDiving=~
      bIsParachuteOpen=~
      bIsSkydiving=~
      bIsParachuteLeaning=~
      bIsSkydiveLeaning=~
      bIsLeaning=~
      bIsSkydiveDiveMode=~
      bPlaySkydiveDrift=~
      bSkydiveDriftDelayActive=~
      bSkydiveDriftAnimAllowed=~
      bIsGliderRight=~
      bIsGliderCenter=~
      bIsGliderLeft=~
      bIsGliderForward=~
      bIsGliderBack=~
      bHasDeployedAndIsStandingUpright=~
      LocalAccelForward=~
      LocalAccelRight=~
      LocalVelocityRight=~
      SkydiveAimPitch=~
      SkydiveAimYaw=~
      DeployChuteAnimRate=~
      SkydiveDriftAnimRate=~
      SkydiveDriftAnimRateCurrent=~
      SkydiveFidgetAnimRate=~
      SkydiveFidgetAnimRateCurrent=~
      SkydiveAdditiveAlpha=~
      SkydiveDriftDelay=~
     ).bIsGliderForward))
  ]
 }
 Transition {path=AnimGraph.AnimGraphNode_StateMachine_1.New State Machine_1.AnimStateTransitionNode_3.Transition, nodes [Result()]}
 Transition {
  path=AnimGraph.AnimGraphNode_StateMachine_1.New State Machine_1.AnimStateTransitionNode_4.Transition
  nodes [Result(bCanEnterTransition=Break Fort Anim Input Skydiving(FortAnimInput_Skydiving=Get SkydivingState()) -> (
     ParaGilderRootOffset=~
     ParaGliderRootRotationOverride=~
     bUseParaGlideRootModifier=~
     bIsSkydivingFromLaunchPad=~
     bIsSkydivingFromBus=~
     bIsInVortex=~
     bIsDiving=~
     bIsParachuteOpen=~
     bIsSkydiving=~
     bIsParachuteLeaning=~
     bIsSkydiveLeaning=~
     bIsLeaning=~
     bIsSkydiveDiveMode=~
     bPlaySkydiveDrift=~
     bSkydiveDriftDelayActive=~
     bSkydiveDriftAnimAllowed=~
     bIsGliderRight=~
     bIsGliderCenter=~
     bIsGliderLeft=~
     bIsGliderForward=~
     bIsGliderBack=~
     bHasDeployedAndIsStandingUpright=~
     LocalAccelForward=~
     LocalAccelRight=~
     LocalVelocityRight=~
     SkydiveAimPitch=~
     SkydiveAimYaw=~
     DeployChuteAnimRate=~
     SkydiveDriftAnimRate=~
     SkydiveDriftAnimRateCurrent=~
     SkydiveFidgetAnimRate=~
     SkydiveFidgetAnimRateCurrent=~
     SkydiveAdditiveAlpha=~
     SkydiveDriftDelay=~
    ).bIsGliderForward)]
 }
 FrontEnd {
  entry=Closed
  path=AnimGraph.AnimGraphNode_StateMachine_2.FrontEnd
  ids=id
  nodes [
   AnimStateNode_0 = Into() {bAlwaysResetOnEntry=true}
   AnimStateNode_1 = Loop() {is=title AnimStateNode}
   AnimStateNode_2 = Closed()
   AnimStateTransitionNode_0 = AnimStateTransitionNode() {bAutomaticRuleBasedOnSequencePlayerInState=true, crossfadeDuration=0}
   AnimStateTransitionNode_2 = AnimStateTransitionNode()
   AnimStateTransitionNode_3 = AnimStateTransitionNode() {crossfadeDuration=0}
   AnimStateTransitionNode_4 = AnimStateTransitionNode() {crossfadeDuration=0}
  ]
  links [
   {
    dst=Into
    rule=AnimGraph.AnimGraphNode_StateMachine_2.FrontEnd.AnimStateTransitionNode_2.Transition
    src=Closed
    via=AnimStateTransitionNode_2
   }
   {
    dst=Closed
    rule=AnimGraph.AnimGraphNode_StateMachine_2.FrontEnd.AnimStateTransitionNode_4.Transition
    src=Into
    via=AnimStateTransitionNode_4
   }
   {dst=Loop, rule=AnimGraph.AnimGraphNode_StateMachine_2.FrontEnd.AnimStateTransitionNode_0.Transition, src=Into, via=AnimStateTransitionNode_0}
   {
    dst=Closed
    rule=AnimGraph.AnimGraphNode_StateMachine_2.FrontEnd.AnimStateTransitionNode_3.Transition
    src=Loop
    via=AnimStateTransitionNode_3
   }
  ]
 }
 Into {
  path=AnimGraph.AnimGraphNode_StateMachine_2.FrontEnd.AnimStateNode_0.Into
  nodes [Output Animation Pose(Result=Sequence Player 'Logarithm_Into_FrontEnd_GLIDER'!(bLoopAnimation:bool=False) -> (Pose) {bLoopAnimation=false, sequence=/BRCosmetics/Animation/Game/MainPlayer/Skydive/ParaGlide/Logarithm/Logarithm_Into_FrontEnd_GLIDER})]
 }
 Loop {
  path=AnimGraph.AnimGraphNode_StateMachine_2.FrontEnd.AnimStateNode_1.Loop
  nodes [Output Animation Pose(Result=Sequence Player 'Logarithm_Glide_FrontEnd_GLIDER'() {sequence=/BRCosmetics/Animation/Game/MainPlayer/Skydive/ParaGlide/Logarithm/Logarithm_Glide_FrontEnd_GLIDER})]
 }
 Closed {
  path=AnimGraph.AnimGraphNode_StateMachine_2.FrontEnd.AnimStateNode_2.Closed
  nodes [Output Animation Pose(Result=Sequence Player 'Logarithm_Closed_GLIDER'() {sequence=&Logarithm_Closed_GLIDER})]
 }
 Transition {path=AnimGraph.AnimGraphNode_StateMachine_2.FrontEnd.AnimStateTransitionNode_0.Transition, nodes [Result()]}
 Transition {
  path=AnimGraph.AnimGraphNode_StateMachine_2.FrontEnd.AnimStateTransitionNode_2.Transition
  nodes [
   n1 = Get SkydivingState()
   Result(bCanEnterTransition=OR Boolean(
     A=Break Fort Anim Input Skydiving(FortAnimInput_Skydiving=@n1) -> (
      ParaGilderRootOffset=~
      ParaGliderRootRotationOverride=~
      bUseParaGlideRootModifier=~
      bIsSkydivingFromLaunchPad=~
      bIsSkydivingFromBus=~
      bIsInVortex=~
      bIsDiving=~
      bIsParachuteOpen=~
      bIsSkydiving=~
      bIsParachuteLeaning=~
      bIsSkydiveLeaning=~
      bIsLeaning=~
      bIsSkydiveDiveMode=~
      bPlaySkydiveDrift=~
      bSkydiveDriftDelayActive=~
      bSkydiveDriftAnimAllowed=~
      bIsGliderRight=~
      bIsGliderCenter=~
      bIsGliderLeft=~
      bIsGliderForward=~
      bIsGliderBack=~
      bHasDeployedAndIsStandingUpright=~
      LocalAccelForward=~
      LocalAccelRight=~
      LocalVelocityRight=~
      SkydiveAimPitch=~
      SkydiveAimYaw=~
      DeployChuteAnimRate=~
      SkydiveDriftAnimRate=~
      SkydiveDriftAnimRateCurrent=~
      SkydiveFidgetAnimRate=~
      SkydiveFidgetAnimRateCurrent=~
      SkydiveAdditiveAlpha=~
      SkydiveDriftDelay=~
     ).bIsParachuteOpen
     B=Get bForceOpen()
    ))
  ]
 }
 Transition {
  path=AnimGraph.AnimGraphNode_StateMachine_2.FrontEnd.AnimStateTransitionNode_3.Transition
  nodes [
   n1 = Get bForceOpen()
   n2 = Break Fort Anim Input Skydiving(FortAnimInput_Skydiving=Get SkydivingState()) -> (
    ParaGilderRootOffset=~
    ParaGliderRootRotationOverride=~
    bUseParaGlideRootModifier=~
    bIsSkydivingFromLaunchPad=~
    bIsSkydivingFromBus=~
    bIsInVortex=~
    bIsDiving=~
    bIsParachuteOpen=~
    bIsSkydiving=~
    bIsParachuteLeaning=~
    bIsSkydiveLeaning=~
    bIsLeaning=~
    bIsSkydiveDiveMode=~
    bPlaySkydiveDrift=~
    bSkydiveDriftDelayActive=~
    bSkydiveDriftAnimAllowed=~
    bIsGliderRight=~
    bIsGliderCenter=~
    bIsGliderLeft=~
    bIsGliderForward=~
    bIsGliderBack=~
    bHasDeployedAndIsStandingUpright=~
    LocalAccelForward=~
    LocalAccelRight=~
    LocalVelocityRight=~
    SkydiveAimPitch=~
    SkydiveAimYaw=~
    DeployChuteAnimRate=~
    SkydiveDriftAnimRate=~
    SkydiveDriftAnimRateCurrent=~
    SkydiveFidgetAnimRate=~
    SkydiveFidgetAnimRateCurrent=~
    SkydiveAdditiveAlpha=~
    SkydiveDriftDelay=~
   )
   Result(bCanEnterTransition=AND Boolean(A=NOT Boolean(A=@n2.bIsParachuteOpen), B=NOT Boolean(A=@n1)))
  ]
 }
 Transition {
  path=AnimGraph.AnimGraphNode_StateMachine_2.FrontEnd.AnimStateTransitionNode_4.Transition
  nodes [
   n1 = Get bForceOpen()
   n2 = Break Fort Anim Input Skydiving(FortAnimInput_Skydiving=Get SkydivingState()) -> (
    ParaGilderRootOffset=~
    ParaGliderRootRotationOverride=~
    bUseParaGlideRootModifier=~
    bIsSkydivingFromLaunchPad=~
    bIsSkydivingFromBus=~
    bIsInVortex=~
    bIsDiving=~
    bIsParachuteOpen=~
    bIsSkydiving=~
    bIsParachuteLeaning=~
    bIsSkydiveLeaning=~
    bIsLeaning=~
    bIsSkydiveDiveMode=~
    bPlaySkydiveDrift=~
    bSkydiveDriftDelayActive=~
    bSkydiveDriftAnimAllowed=~
    bIsGliderRight=~
    bIsGliderCenter=~
    bIsGliderLeft=~
    bIsGliderForward=~
    bIsGliderBack=~
    bHasDeployedAndIsStandingUpright=~
    LocalAccelForward=~
    LocalAccelRight=~
    LocalVelocityRight=~
    SkydiveAimPitch=~
    SkydiveAimYaw=~
    DeployChuteAnimRate=~
    SkydiveDriftAnimRate=~
    SkydiveDriftAnimRateCurrent=~
    SkydiveFidgetAnimRate=~
    SkydiveFidgetAnimRateCurrent=~
    SkydiveAdditiveAlpha=~
    SkydiveDriftDelay=~
   )
   Result(bCanEnterTransition=AND Boolean(A=NOT Boolean(A=@n2.bIsParachuteOpen), B=NOT Boolean(A=@n1)))
  ]
 }
 EventGraph {nodes [
   Event BlueprintUpdateAnimation()
   > n1 = Sequence()
   n2 = Break Fort Anim Input Skydiving(FortAnimInput_Skydiving=Get SkydivingState()) -> (
    ParaGilderRootOffset=~
    ParaGliderRootRotationOverride=~
    bUseParaGlideRootModifier=~
    bIsSkydivingFromLaunchPad=~
    bIsSkydivingFromBus=~
    bIsInVortex=~
    bIsDiving=~
    bIsParachuteOpen=~
    bIsSkydiving=~
    bIsParachuteLeaning=~
    bIsSkydiveLeaning=~
    bIsLeaning=~
    bIsSkydiveDiveMode=~
    bPlaySkydiveDrift=~
    bSkydiveDriftDelayActive=~
    bSkydiveDriftAnimAllowed=~
    bIsGliderRight=~
    bIsGliderCenter=~
    bIsGliderLeft=~
    bIsGliderForward=~
    bIsGliderBack=~
    bHasDeployedAndIsStandingUpright=~
    LocalAccelForward=~
    LocalAccelRight=~
    LocalVelocityRight=~
    SkydiveAimPitch=~
    SkydiveAimYaw=~
    DeployChuteAnimRate=~
    SkydiveDriftAnimRate=~
    SkydiveDriftAnimRateCurrent=~
    SkydiveFidgetAnimRate=~
    SkydiveFidgetAnimRateCurrent=~
    SkydiveAdditiveAlpha=~
    SkydiveDriftDelay=~
   )
   > Branch(Condition=NOT Boolean(A=@n2.bIsGliderForward))
   > Set SkipThrustFwdPlay(SkipThrustFwdPlay:bool=false) -> (Output_Get:bool)
   Reroute Node(InputPin=@n1.then_1)
   AnimNotify_TurnONEngines() {is=title K2Node_Event}
   > SetGliderFullyDeployed!(execute, self:Fort Glider Anim Instance Object Reference, bIsFullyDeployed:bool=true) -> (then)
   AnimNotify_Idle_Loop_Entered() {is=title K2Node_Event}
   > Set SkipThrustFwdPlay(SkipThrustFwdPlay=@n2.bIsGliderForward)
   n3 = Break Fort Anim Input Skydiving(FortAnimInput_Skydiving=Get SkydivingState()) -> (
    ParaGilderRootOffset=~
    ParaGliderRootRotationOverride=~
    bUseParaGlideRootModifier=~
    bIsSkydivingFromLaunchPad=~
    bIsSkydivingFromBus=~
    bIsInVortex=~
    bIsDiving=~
    bIsParachuteOpen=~
    bIsSkydiving=~
    bIsParachuteLeaning=~
    bIsSkydiveLeaning=~
    bIsLeaning=~
    bIsSkydiveDiveMode=~
    bPlaySkydiveDrift=~
    bSkydiveDriftDelayActive=~
    bSkydiveDriftAnimAllowed=~
    bIsGliderRight=~
    bIsGliderCenter=~
    bIsGliderLeft=~
    bIsGliderForward=~
    bIsGliderBack=~
    bHasDeployedAndIsStandingUpright=~
    LocalAccelForward=~
    LocalAccelRight=~
    LocalVelocityRight=~
    SkydiveAimPitch=~
    SkydiveAimYaw=~
    DeployChuteAnimRate=~
    SkydiveDriftAnimRate=~
    SkydiveDriftAnimRateCurrent=~
    SkydiveFidgetAnimRate=~
    SkydiveFidgetAnimRateCurrent=~
    SkydiveAdditiveAlpha=~
    SkydiveDriftDelay=~
   )
   "To String (Float)"(InDouble=@n3.LocalAccelForward)
   "To String (Float)"(InDouble=GetCurveValue(CurveName=Idle_Loop_Additive))
   n4 = NOT Boolean(A=Get bForceOpen())
   n5 = NOT Boolean(A=Break Fort Anim Input Skydiving(FortAnimInput_Skydiving=Get SkydivingState()) -> (
     ParaGilderRootOffset=~
     ParaGliderRootRotationOverride=~
     bUseParaGlideRootModifier=~
     bIsSkydivingFromLaunchPad=~
     bIsSkydivingFromBus=~
     bIsInVortex=~
     bIsDiving=~
     bIsParachuteOpen=~
     bIsSkydiving=~
     bIsParachuteLeaning=~
     bIsSkydiveLeaning=~
     bIsLeaning=~
     bIsSkydiveDiveMode=~
     bPlaySkydiveDrift=~
     bSkydiveDriftDelayActive=~
     bSkydiveDriftAnimAllowed=~
     bIsGliderRight=~
     bIsGliderCenter=~
     bIsGliderLeft=~
     bIsGliderForward=~
     bIsGliderBack=~
     bHasDeployedAndIsStandingUpright=~
     LocalAccelForward=~
     LocalAccelRight=~
     LocalVelocityRight=~
     SkydiveAimPitch=~
     SkydiveAimYaw=~
     DeployChuteAnimRate=~
     SkydiveDriftAnimRate=~
     SkydiveDriftAnimRateCurrent=~
     SkydiveFidgetAnimRate=~
     SkydiveFidgetAnimRateCurrent=~
     SkydiveAdditiveAlpha=~
     SkydiveDriftDelay=~
    ).bIsParachuteOpen)
   PrintString!(execute, InString="To String (Boolean)"(InBool=AND Boolean(A=@n5, B=@n4)), bPrintToScreen, bPrintToLog, TextColor, Duration, Key) -> (then)
   "To String (Boolean)"(InBool=@n3.bIsGliderForward)
   TryGetPawnOwner()
  ]}
]
variables [
 bAllowBlendToFwdADD {type=boolean}
 SkipThrustFwdPlay {type=boolean}
]
